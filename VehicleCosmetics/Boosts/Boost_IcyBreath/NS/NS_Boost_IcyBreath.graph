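# Niagara system for the Icy Breath vehicle boost cosmetic, emitting icy snow and smoke effects from filtered sockets on a skeletal mesh.
# A sprite/mesh emitter spawns rear-moving particles in several spawn groups, a ribbon emitter lays three-layer snow trails, and STATIC_MESH holds looping ice and flame meshes with sine wobble.
# Painted-variant overrides replace default icy colors, which mostly vary by GameModeID; inherited source movement keeps trails attached.
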NiagaraSystem /VehicleCosmetics/Boosts/Boost_IcyBreath/NS/NS_Boost_IcyBreath
dependencies {
 usedDataInterfaces [
  &NiagaraDataInterfaceCurve
  &NiagaraDataInterfaceSpriteRendererInfo
  &NiagaraDataInterfaceSkeletalMesh
 ]
 usedDynamicInputs []
 usedModules [
  &self:SystemSpawnScript.SystemScriptSource.SystemScriptGraph.NiagaraNodeAssignment_0.SetVariables_#1
  &SystemState
 ]
}
emitters {
 NE_VC_Boost_Sprite_Parent {
  emitterUpdateScript [
   &EmitterState()
   &VC_FilteredSocketCount()
   &SpawnRate(SpawnRate=150, Use Spawn Probability=false)
   &SpawnRate(SpawnGroup=1, SpawnRate=200, Use Spawn Probability=false) {module=SpawnRate001}
   &SpawnRate(SpawnGroup=3, SpawnRate=250, Use Spawn Probability=false) {module=SpawnRate003}
  ]
  particleSpawnScript [
   &VC_PaintedVariantsOverride(
    Black=<rgba 0.03 0.03 0.03 3>
    BurntSienna=<rgba 0.15 0.059 0.015 2.5>
    Cobalt=<rgba 0 0.15 1 1>
    Crimson=<rgba 0.849 0.2137 0.2309 1>
    ForestGreen=<rgba 0 0.5 0.0253 1>
    Grey=<rgba 0.3073 0.3073 0.3073 1>
    Lime=<rgba 0.4175 0.8351 0 1>
    Orange=<rgba 1.15 0.4011 0 1>
    Pink=<rgba 0.8108 0.4546 0.6813 1>
    Purple=<rgba 0.1525 0.015 0.35 2>
    R1 Black:LinearColor=<rgba 0.14 0.14 0.14 3>
    R1 BurntSienna:LinearColor=<rgba 0.4045 0.1814 0.1215 2>
    R1 Cobalt:LinearColor=<rgba 0.2415 0.3297 1.3 2>
    R1 Crimson:LinearColor=<rgba 0.724 0.1395 0.1554 3>
    R1 ForestGreen:LinearColor=<rgba 0.0841 0.8212 0.1214 3>
    R1 Grey:LinearColor=<rgba 0.4232 0.4232 0.4232 3>
    R1 Lime:LinearColor=<rgba 0.7075 1 0.4149 2.5>
    R1 Orange:LinearColor=<rgba 0.7865 0.4032 0.198 2>
    R1 Pink:LinearColor=<rgba 1 0.4149 0.7872 2>
    R1 Purple:LinearColor=<rgba 0.4299 0.033 1 2>
    R1 Saffron:LinearColor=<rgba 0.7568 0.8733 0 2>
    R1 SkyBlue:LinearColor=<rgba 0.2344 0.7438 1 2>
    R2 Black:LinearColor=<rgba 0 0 0 1>
    R2 BurntSienna:LinearColor=<rgba 0.224 0.0967 0.0626 1>
    R2 Cobalt:LinearColor=<rgba 0.1466 0.2967 0.5816 1>
    R2 Crimson:LinearColor=<rgba 0.4635 0 0.0126 1>
    R2 ForestGreen:LinearColor=<rgba 0 0.35 0.0177 1>
    R2 Grey:LinearColor=<rgba 0.276 0.276 0.276 1>
    R2 Lime:LinearColor=<rgba 0.3913 0.4957 0.0991 1>
    R2 Orange:LinearColor=<rgba 0.6267 0.2732 0.0838 1>
    R2 Pink:LinearColor=<rgba 0.4358 0.2444 0.3662 1>
    R2 Purple:LinearColor=<rgba 0.1104 0 0.2691 1>
    R2 Saffron:LinearColor=<rgba 0.4392 0.4392 0.0999 1>
    R2 SkyBlue:LinearColor=<rgba 0.1503 0.4835 0.651 1>
    R3 Black:LinearColor=<rgba 0.65 0.65 0.65 1>
    R3 BurntSienna:LinearColor=<rgba 0.7101 0.538 0.4919 1>
    R3 Cobalt:LinearColor=<rgba 0.6927 0.7695 1 1>
    R3 Crimson:LinearColor=<rgba 1 0.6441 0.6441 1>
    R3 ForestGreen:LinearColor=<rgba 0.4539 0.8628 0.4746 0>
    R3 Grey:LinearColor=<rgba 0.7135 0.7135 0.7135 1>
    R3 Lime:LinearColor=<rgba 0.9262 1 0.8524 0>
    R3 Orange:LinearColor=<rgba 1 0.7347 0.3366 0.5>
    R3 Pink:LinearColor=<rgba 1 0.5608 0.8403 0>
    R3 Purple:LinearColor=<rgba 0.6525 0.4384 0.9601 1>
    R3 Saffron:LinearColor=<rgba 0.9033 1 0.6927 0>
    R3 SkyBlue:LinearColor=<rgba 0.6615 0.8867 1 1>
    Saffron=<rgba 0.6291 0.7969 0.1259 1>
    SkyBlue=<rgba 0 0.6642 0.9983 0.5>
    TitaniumWhite=<rgba 0.75 0.75 0.75 0.5>
    Use R1 Painted Overrides=true
    Use R2 Painted Overrides=true
    Use R3 Painted Overrides=true
   )
   &InitializeParticle(
    Color Mode=::NewEnumerator0
    Lifetime=&MergeHelper_FloatSwitch(
     0=&RandomRangeFloat(Minimum=0.5, Maximum=0.2)
     1=0.55
     2=0.2
     3=0.2
     Index=$Emitter.SpawnGroup
     Num=::NewEnumerator2
    )
    Mesh Scale Mode=::NewEnumerator3
    Mesh Uniform Scale=&MergeHelper_FloatSwitch(
     0=0
     1=&RandomRangeFloat(Minimum=0.3, Maximum=0.15)
     2=&RandomRangeFloat(Minimum=0.5, Maximum=0.3)
     3=0.04
     Index=$Emitter.SpawnGroup
     Num=::NewEnumerator2
    )
    Sprite Rotation Mode=::NewEnumerator2
    Sprite Size Mode=::NewEnumerator3
    Uniform Sprite Size=50
    UsePositionOffset=false
    Write Material Random=true
   )
   &PartitionParticles(Number Of Partitions=$Emitter.PortCount)
   &SkeletalMeshLocation(
    Bone / Socket Index=$Particles.PartitionParticles.Partition
    Bone / Socket Sampling Mode=::NewEnumerator5
    MaskSpawnGroup=false
    Mesh Sampling Type=::NewEnumerator6
    Sampled Position Offset=&MergeHelper_VectorSwitch(0=<-100 0 0>, 1=<-75 0 0>, 2=<-100 0 0>, 3=<-90 0 0>, Index=$Emitter.SpawnGroup, Num=::NewEnumerator2)
    Skeletal Mesh=$System.SkeletalMesh
   )
   &AddVelocity(Velocity=&MergeHelper_VectorSwitch(0=<-350 0 0>, 1=<-1000 0 0>, 2=<-1000 0 0>, 3=<-850 0 0>, Index=$Emitter.SpawnGroup, Num=::NewEnumerator2))
   /SetVariables(Particles.VisibilityTag=$Emitter.SpawnGroup) {isSetParameters=true}
   &ShapeLocation(
    MaskSpawnGroup=false
    Override Seed=false
    Sphere Radius=&MergeHelper_FloatSwitch(0=20, 1=35, 2=35, 3=0, Index=$Emitter.SpawnGroup, Num=::NewEnumerator2)
    UseOwnerScale=true
   )
   &InitialMeshOrientation(Use Rotation Vector=false)
  ]
  particleUpdateScript [
   /SetVariables(
    Particles.BC Pices:LinearColor=&MakeCustomLinearColorFromBool(
     Bool=$User.bIsPainted
     True Color=$Particles.R1 Color
     False Color=&MergeHelper_ColorSwitch(0=<rgba 0.6724 0.7816 1 1>, 1=<rgba 1.0087 1.1724 1.5 2>, 2=<rgba 1.0087 1.1724 1.5 2>, 3=?, Index=$User.GameModeID, Num=::NewEnumerator1)
    )
    Particles.BC Smoke:LinearColor=&MakeCustomLinearColorFromBool(
     Bool=$User.bIsPainted
     True Color=$Particles.R2 Color
     False Color=&MergeHelper_ColorSwitch(0=<rgba 0.3448 0.4214 0.5747 1>, 1=<rgba 0.54 0.66 0.9 1>, 2=<rgba 0.4264 0.5509 0.8 1>, 3=?, Index=$User.GameModeID, Num=::NewEnumerator1)
    )
    Particles.BC Smoke Fresnel:LinearColor=&MakeCustomLinearColorFromBool(
     Bool=$User.bIsPainted
     True Color=&MultiplyLinearColor(
      Linear Color=$Particles.R3 Color
      Scale Factor=&MakeVector4FromVectorAndFloat("Vector (RGB)"=<1 1 1>, "Float (Alpha)"=1)
     )
     False Color=&MultiplyLinearColor(
      Linear Color=<rgba 1 1 1 1>
      Scale Factor=&MakeVector4FromVectorAndFloat("Vector (RGB)"=<1 1 1>, "Float (Alpha)"=0)
     )
    )
    Particles.BC Snow Wind:LinearColor=&MakeCustomLinearColorFromBool(
     Bool=$User.bIsPainted
     True Color=&MultiplyLinearColor(
      Linear Color=$Particles.Color
      Scale Factor=&MakeVector4FromVectorAndFloat(
       "Vector (RGB)"=<1 1 1>
       "Float (Alpha)"=&FloatFromCurve(FloatCurve=DataInterface() {propertyValues {Curve {
           defaultValue=1
           keys [
            {arriveTangent=5.21, interpMode=RCIM_Cubic, leaveTangent=5.21, tangentMode=RCTM_User, time=0, value=0}
            {interpMode=RCIM_Cubic, tangentMode=RCTM_User, time=0.362, value=1}
            {arriveTangent=-0.2199, tangentMode=RCTM_User, time=1, value=0}
           ]
          }}})
      )
     )
     False Color=&MultiplyLinearColor(
      Linear Color=&MergeHelper_ColorSwitch(0=<rgba 0.467 0.7116 1 1>, 1=<rgba 0.5838 0.8896 1.25 1>, 2=<rgba 0.5447 0.8684 1.25 1>, 3=?, Index=$User.GameModeID, Num=::NewEnumerator1)
      Scale Factor=&MakeVector4FromVectorAndFloat(
       "Vector (RGB)"=<1 1 1>
       "Float (Alpha)"=&FloatFromCurve(FloatCurve=DataInterface() {propertyValues {Curve {
           defaultValue=1
           keys [
            {arriveTangent=5.21, interpMode=RCIM_Cubic, leaveTangent=5.21, tangentMode=RCTM_User, time=0, value=0}
            {interpMode=RCIM_Cubic, tangentMode=RCTM_User, time=0.362, value=1}
            {arriveTangent=-0.2199, tangentMode=RCTM_User, time=1, value=0}
           ]
          }}})
      )
     )
    )
    Particles.Fast pieces scale:Vector3f=&VectorFromFloat(Value=&FloatFromCurve(
      FloatCurve=DataInterface() {propertyValues {Curve {keys [<at 0.3 2>,<at 0.31 1>]}}}
      Scale Curve=&RandomRangeFloat(Minimum=0.25, Maximum=0.2)
     ))
    Particles.Mesh size Smoke:Vector3f=&VectorFromFloat(Value=&FloatFromCurve(
      FloatCurve=DataInterface() {propertyValues {Curve {
         defaultValue=1
         keys [
          {arriveTangent=-0.7718, interpMode=RCIM_Cubic, time=0.25, value=1}
          {arriveTangentWeight=0.1692, interpMode=RCIM_Cubic, tangentMode=RCTM_User, time=1, value=0.4}
         ]
        }}}
      Scale Curve=&RandomRangeFloat(Minimum=0.025, Maximum=0.026)
     ))
   ) {isSetParameters=true}
   &ParticleState()
   &ScaleSpriteSize(Uniform Curve Sprite Scale=DataInterface() {propertyValues {Curve {
       defaultValue=0
       keys [
        {interpMode=RCIM_Cubic, tangentMode=RCTM_User, time=0, value=0.2}
        {interpMode=RCIM_Cubic, tangentMode=RCTM_User, time=0.5, value=1}
        {tangentMode=RCTM_User, time=1, value=0.45}
       ]
      }}})
   &InheritSourceMovement(
    Applied Position Scale=&Multiply_VectorByFloat(
     Vector=<1 1 1>
     Float=&FloatFromCurve(
      FloatCurve=DataInterface() {propertyValues {Curve {
         defaultValue=1
         keys [
          {tangentMode=RCTM_User, time=0, value=1}
          {arriveTangent=-0.0405, interpMode=RCIM_Cubic, leaveTangent=-0.0405, tangentMode=RCTM_User, time=0.7009, value=1}
          {arriveTangent=-8.0089, interpMode=RCIM_Cubic, leaveTangent=-8.0089, tangentMode=RCTM_User, time=1, value=0}
         ]
        }}}
      Scale Curve=1.1
     )
    )
    ApplyForce=false
    ApplyPosition=true
    ApplyPositionFalloff=false
   )
   &SolveForcesAndVelocity(Clamp Velocity=false, Limit Acceleration=false)
   &SubUVAnimation(
    "Sprite Renderer "=DataInterface() {propertyValues {SpriteRenderer=&self:NE_VC_Boost_Sprite_Parent.NiagaraSpriteRendererProperties_0}}
    UseEndFrame=false
    UseStartFrame=false
   )
  ]
  rendererClasses [&NiagaraMeshRendererProperties, &NiagaraSpriteRendererProperties]
  renderers [
   {rendererClass=&NiagaraMeshRendererProperties, rendererIndex=0}
   {rendererClass=&NiagaraMeshRendererProperties, rendererIndex=1}
   {rendererClass=&NiagaraMeshRendererProperties, rendererIndex=2}
   {rendererClass=&NiagaraSpriteRendererProperties, rendererIndex=3}
  ]
 }
 Ribbon {
  emitterUpdateScript [
   &EmitterState()
   &VC_FilteredSocketCount()
   &VC_RateQualityHandler(Epic Spawn Rate=20, High Spawn Rate=20, Low Spawn Rate=20, Medium Spawn Rate=20, Use Custom Rate?=true)
   &VC_SpawnRibbon_v1(Number of Ribbons=$Emitter.PortCount, Spawn Time=-1, SpawnRate=$Emitter.Spawn Rate)
  ]
  particleSpawnScript [
   &VC_PaintedVariantsOverride(
    Black=<rgba 0 0 0 2>
    BurntSienna=<rgba 0.5052 0.2417 0.171 1>
    Cobalt=<rgba 0 0.0222 1.2 1>
    Crimson=<rgba 0.4635 0 0.0126 1>
    ForestGreen=<rgba 0 0.5 0.0253 1>
    Grey=<rgba 0.4358 0.4358 0.4358 1>
    Orange=<rgba 1 0.5082 0.2448 1>
    Pink=<rgba 1 0.3351 0.7582 1>
    Purple=<rgba 0.2789 0 0.6795 1>
    R1 Black:LinearColor=<rgba 0.276 0.276 0.276 1>
    R1 BurntSienna:LinearColor=<rgba 0.3385 0.2146 0.1675 1>
    R1 Cobalt:LinearColor=<rgba 0.1725 0.3572 0.9115 2>
    R1 Crimson:LinearColor=<rgba 0.6372 0.2622 0.2723 1>
    R1 ForestGreen:LinearColor=<rgba 0.3844 0.6649 0.3986 1>
    R1 Grey:LinearColor=<rgba 0.2274 0.1959 0.0944 1>
    R1 Lime:LinearColor=<rgba 0.7431 0.9288 0 1>
    R1 Orange:LinearColor=<rgba 1 0.3488 0 2>
    R1 Pink:LinearColor=<rgba 1 0.5087 0.8213 1.5>
    R1 Purple:LinearColor=<rgba 0.3241 0.0975 0.65 3>
    R1 Saffron:LinearColor=<rgba 0.75 1 0 1>
    R1 SkyBlue:LinearColor=<rgba 0.246 0.668 0.8802 2>
    R2 Black:LinearColor=<rgba 0.15 0.15 0.15 2>
    R2 BurntSienna:LinearColor=<rgba 0.3739 0.2073 0.0506 3>
    R2 Cobalt:LinearColor=<rgba 0.1892 0.2042 1 2>
    R2 Crimson:LinearColor=<rgba 0.5122 0 0.0139 3>
    R2 ForestGreen:LinearColor=<rgba 0 0.5333 0.0269 4>
    R2 Grey:LinearColor=<rgba 0.5391 0.5391 0.5391 1.5>
    R2 Lime:LinearColor=<rgba 0.5327 0.7795 0 3>
    R2 Orange:LinearColor=<rgba 0.8316 0.4486 0 3>
    R2 Pink:LinearColor=<rgba 0.5052 0.3693 0.4558 3>
    R2 Purple:LinearColor=<rgba 0.2873 0 0.7 2>
    R2 Saffron:LinearColor=<rgba 0.7705 0.8406 0 3>
    R2 SkyBlue:LinearColor=<rgba 0 0.6111 0.9184 2.5>
    SkyBlue=<rgba 0.2947 0.5919 0.7413 1>
    Use R1 Painted Overrides=true
    Use R2 Painted Overrides=true
   )
   &VC_InitializeRibbon_v1(
    Color=$Particles.Color
    Custom Offset=true
    Enable Skeletal Position=true
    Lifetime=1
    Local Vector Offset=&AddVectorToPosition(
     Position=&MergeHelper_VectorSwitch(0=<-100 0 0>, 1=<-100 0 0>, 2=<-50 0 0>, 3=?, Index=$User.GameModeID, Num=::NewEnumerator1)
     Vector=$User.DTCarOffset
    )
    Offset Amount=1
    Ribbon Width=43
   )
   &AddVelocity(Velocity=<-1500 0 0>) {module=AddVelocity001}
  ]
  particleUpdateScript [
   /SetVariables(Particles.BC_Complement:LinearColor=&MakeCustomLinearColorFromBool(Bool=$User.bIsPainted, True Color=$Particles.Color, False Color=<rgba 0.6 0.7836 1 1>)) {isSetParameters=true}
   /SetVariables(Particles.Size Backgraund Snow=30, Particles.Size Particles Snow=25, Particles.Size Wind Snow=50) {isSetParameters=true}
   /SetVariables(
    Particles.BC Backgraund Snow:LinearColor=&MakeCustomLinearColorFromBool(
     Bool=$User.bIsPainted
     True Color=&MultiplyLinearColor(
      Linear Color=$Particles.R2 Color
      Scale Factor=&MakeVector4FromVectorAndFloat("Vector (RGB)"=<1 1 1>, "Float (Alpha)"=1)
     )
     False Color=&MultiplyLinearColor(
      Linear Color=&MergeHelper_ColorSwitch(
       0=<rgba 0.5364 0.651 0.8802 1>
       1=<rgba 0.7617 0.9245 1.25 1>
       2=<rgba 0.7617 0.9245 1.25 1>
       3=?
       Index=$User.GameModeID
       Num=::NewEnumerator1
      )
      Scale Factor=&MakeVector4FromVectorAndFloat("Vector (RGB)"=<1 1 1>, "Float (Alpha)"=1)
     )
    )
    Particles.BC Particle Snow:LinearColor=&MakeCustomLinearColorFromBool(
     Bool=$User.bIsPainted
     True Color=&MultiplyLinearColor(
      Linear Color=$Particles.Color
      Scale Factor=&MakeVector4FromVectorAndFloat("Vector (RGB)"=<1 1 1>, "Float (Alpha)"=1)
     )
     False Color=&MultiplyLinearColor(
      Linear Color=<rgba 0.6724 0.8228 1 1>
      Scale Factor=&MakeVector4FromVectorAndFloat("Vector (RGB)"=<1 1 1>, "Float (Alpha)"=1)
     )
    )
    Particles.BC Wind Snow:LinearColor=&MakeCustomLinearColorFromBool(
     Bool=$User.bIsPainted
     True Color=&MultiplyLinearColor(
      Linear Color=$Particles.R1 Color
      Scale Factor=&MakeVector4FromVectorAndFloat("Vector (RGB)"=<1 1 1>, "Float (Alpha)"=1)
     )
     False Color=&MultiplyLinearColor(
      Linear Color=<rgba 0.6098 0.8783 1 1>
      Scale Factor=&MakeVector4FromVectorAndFloat("Vector (RGB)"=<1 1 1>, "Float (Alpha)"=1)
     )
    )
   ) {isSetParameters=true}
   &ParticleState()
   &InheritSourceMovement(
    Applied Position Scale=&Multiply_VectorByFloat(
     Vector=<1 1 1>
     Float=&FloatFromCurve(
      FloatCurve=DataInterface() {propertyValues {Curve {
         defaultValue=0
         keys [
          {tangentMode=RCTM_User, time=0, value=1}
          {arriveTangent=-0.0405, interpMode=RCIM_Cubic, leaveTangent=-0.0405, tangentMode=RCTM_User, time=0.7009, value=1}
          {arriveTangent=-8.0089, interpMode=RCIM_Cubic, leaveTangent=-8.0089, tangentMode=RCTM_User, time=1, value=0}
         ]
        }}}
      Scale Curve=1.1
     )
    )
    ApplyForce=false
    ApplyPosition=true
    ApplyPositionFalloff=false
   )
   &SolveForcesAndVelocity(Clamp Velocity=false, Limit Acceleration=false)
   &VC_UpdateRibbon_v1(Enable Skeletal Position=true)
  ]
  rendererClasses [&NiagaraRibbonRendererProperties]
  renderers [
   {rendererClass=&NiagaraRibbonRendererProperties}
   {rendererClass=&NiagaraRibbonRendererProperties, rendererIndex=1}
   {rendererClass=&NiagaraRibbonRendererProperties, rendererIndex=2}
  ]
 }
 STATIC_MESH {
  emitterUpdateScript [
   &EmitterState(
    Inactive Response=::NewEnumerator1
    Life Cycle Mode=::NewEnumerator1
    Loop Behavior=::NewEnumerator1
    Loop Duration Mode=::NewEnumerator1
    UseLoopDelay=false
   )
   &VC_FilteredSocketCount()
   &SpawnBurst_Instantaneous(Spawn Count=2, Use Spawn Probability=false, UseLoopCountLimit=false)
   &SpawnRate(SpawnRate=2.2, Use Spawn Probability=false) {enabled=false}
  ]
  particleSpawnScript [
   &VC_PaintedVariantsOverride(
    Black=<rgba 0.12 0.12 0.12 1>
    BurntSienna=<rgba 0.3976 0.1641 0.1015 1>
    Cobalt=<rgba 0.1973 0.3683 0.8812 1>
    Crimson=<rgba 0.6406 0 0.0174 1>
    ForestGreen=<rgba 0.0525 0.6719 0.0838 1>
    Grey=<rgba 0.3802 0.3802 0.3802 1>
    Lime=<rgba 0.6686 1 0.2101 1>
    Orange=<rgba 1 0.3814 0.05 1>
    Pink=<rgba 0.7101 0.3291 0.5716 1>
    Purple=<rgba 0.331 0.072 0.7031 1>
    R1 Black:LinearColor=<rgba 0.2691 0.2691 0.2691 1>
    R1 BurntSienna:LinearColor=<rgba 0.6649 0.2763 0.172 1>
    R1 Cobalt:LinearColor=<rgba 0.319 0.3362 1.25 1>
    R1 Crimson:LinearColor=<rgba 0.974 0 0.0264 1>
    R1 ForestGreen:LinearColor=<rgba 0.1788 1 0.2203 1>
    R1 Grey:LinearColor=<rgba 0.5781 0.5781 0.5781 1>
    R1 Lime:LinearColor=<rgba 0.5929 1 0.1858 1>
    R1 Orange:LinearColor=<rgba 1 0.4788 0.1997 1>
    R1 Pink:LinearColor=<rgba 0.901 0.5585 0.7765 1>
    R1 Purple:LinearColor=<rgba 0.4104 0 1 1>
    R1 Saffron:LinearColor=<rgba 1 1 0 1>
    R1 SkyBlue:LinearColor=<rgba 0.3142 0.7705 1 1>
    R2 Black:LinearColor=<rgba 0.2 0.2 0.2 1>
    R2 BurntSienna:LinearColor=<rgba 0.349 0.1899 0.1472 1>
    R2 Cobalt:LinearColor=<rgba 0.2114 0.2211 0.7378 1>
    R2 Crimson:LinearColor=<rgba 0.4427 0.1437 0.1518 1>
    R2 ForestGreen:LinearColor=<rgba 0 1 0.0505 1>
    R2 Grey:LinearColor=<rgba 0.3108 0.3108 0.3108 1>
    R2 Lime:LinearColor=<rgba 0.5 1 0 1>
    R2 Pink:LinearColor=<rgba 0.5017 0.2099 0.3956 1>
    R2 Purple:LinearColor=<rgba 0.3408 0.105 0.6795 1>
    R2 Saffron:LinearColor=<rgba 0.9 0.7407 0.2172 1>
    R2 SkyBlue:LinearColor=<rgba 0.1413 0.4805 0.651 1>
    Saffron=<rgba 0.7552 0.7552 0.1062 1>
    SkyBlue=<rgba 0.1319 0.6252 0.8733 1>
    Use R1 Painted Overrides=true
    Use R2 Painted Overrides=true
   )
   /SetVariables(Particles.VisibilityTag=$Emitter.SpawnGroup) {isSetParameters=true}
   &InitializeParticle(
    Color=&MakeCustomLinearColorFromBool(Bool=$User.bIsPainted, True Color=$Particles.Color, False Color=<rgba 0.4984 0.6523 0.9601 1>)
    Mesh Scale Mode=::NewEnumerator3
    Mesh Uniform Scale=-3.5
    Sprite Rotation Mode=::NewEnumerator2
    Sprite Size Mode=::NewEnumerator3
    Uniform Sprite Size=300
    UsePositionOffset=false
    Write Material Random=true
   )
   &PartitionParticles(Number Of Partitions=$Emitter.PortCount)
   &SkeletalMeshLocation(
    Bone / Socket Index=$Particles.PartitionParticles.Partition
    Bone / Socket Sampling Mode=::NewEnumerator5
    MaskSpawnGroup=false
    Mesh Sampling Type=::NewEnumerator6
    Sampled Position Offset=&AddVectorToPosition(Position=<-50 0 0>, Vector=$User.DTCarOffset)
    Skeletal Mesh=$System.SkeletalMesh
   )
   &AddVelocity(Velocity=<0 0 0>) {enabled=false}
  ]
  particleUpdateScript [
   &ParticleState(Kill Particles When Lifetime Has Elapsed=false, Loop Particles Lifetime=true)
   &SolveForcesAndVelocity(Clamp Velocity=false, Limit Acceleration=false) {enabled=false}
   /SetVariables(
    Particles.Rotate Flame:Quat4f=<0 180 0 1>
    Particles.Scale Ice Complement:Vector3f=&VectorFromFloat(Value=-3.8)
    Particles.Scale Litle Ice Complement:Vector3f=&VectorFromFloat(Value=-3.6)
    Particles.Scale_Flame:Vector3f=&VectorFromFloat(Value=3.8)
    Particles.WPO Ice Complemnt:NiagaraPosition=&AddVectorToPosition(
     Position=$Particles.Position
     Vector=&TransformVector(
      Vector=&MakeVector(
       X=-9
       Y=&Sine(Normalized Angle=$Emitter.Age, Period=1, Scale=0.5, Bias=0)
       Z=&Sine(Normalized Angle=$Emitter.Age, Period=1, Scale=0.5, Bias=0)
      )
      Destination Space=::Simulation
      Apply Scale=true
     )
    )
    Particles.WPO Litle Ice Complement:NiagaraPosition=&AddVectorToPosition(
     Position=$Particles.Position
     Vector=&TransformVector(
      Vector=&MakeVector(
       X=-15
       Y=&Sine(Normalized Angle=$Emitter.Age, Period=1, Scale=1, Bias=0)
       Z=&Sine(Normalized Angle=$Emitter.Age, Period=1, Scale=1, Bias=0)
      )
      Destination Space=::Simulation
      Apply Scale=true
     )
    )
    Particles.WPO_Flame:NiagaraPosition=&AddVectorToPosition(Position=$Particles.Position, Vector=<0 5 0>)
    Particles.WPO_Ice:NiagaraPosition=&AddVectorToPosition(
     Position=$Particles.Position
     Vector=&TransformVector(
      Vector=&MakeVector(
       X=-9
       Y=&Sine(Normalized Angle=$Emitter.Age, Period=1, Scale=0.5, Bias=0)
       Z=&Sine(Normalized Angle=$Emitter.Age, Period=1, Scale=0.5, Bias=0)
      )
      Destination Space=::Simulation
      Apply Scale=true
     )
    )
    Particles.WPO_Litle_Ice:NiagaraPosition=&AddVectorToPosition(
     Position=$Particles.Position
     Vector=&TransformVector(
      Vector=&MakeVector(
       X=-15
       Y=&Sine(Normalized Angle=$Emitter.Age, Period=1, Scale=1, Bias=0)
       Z=&Sine(Normalized Angle=$Emitter.Age, Period=1, Scale=1, Bias=0)
      )
      Destination Space=::Simulation
      Apply Scale=true
     )
    )
   ) {isSetParameters=true}
   /SetVariables(
    Particles.BC_Complement:LinearColor=&MakeCustomLinearColorFromBool(Bool=$User.bIsPainted, True Color=$Particles.R2 Color, False Color=<rgba 0.7344 0.7344 0.7344 0.1>)
    Particles.BC_Flame:LinearColor=&MakeCustomLinearColorFromBool(
     Bool=$User.bIsPainted
     True Color=$Particles.R1 Color
     False Color=&MergeHelper_ColorSwitch(
      0=<rgba 0.7138 0.6734 0.9 1>
      1=<rgba 0.9121 0.8605 1.15 1>
      2=<rgba 0.9121 0.8605 1.15 1>
      3=?
      Index=$User.GameModeID
      Num=::NewEnumerator1
     )
    )
    Particles.BC_Ice:LinearColor=&MakeCustomLinearColorFromBool(
     Bool=$User.bIsPainted
     True Color=$Particles.Color
     False Color=&MergeHelper_ColorSwitch(0=<rgba 0.58 0.685 1 1>, 1=<rgba 0.725 0.8562 1.25 1>, 2=<rgba 0.6359 0.7894 1.25 1>, 3=?, Index=$User.GameModeID, Num=::NewEnumerator1)
    )
   ) {isSetParameters=true}
  ]
  rendererClasses [&NiagaraMeshRendererProperties, &NiagaraSpriteRendererProperties]
  renderers [
   {rendererClass=&NiagaraMeshRendererProperties, rendererIndex=0}
   {rendererClass=&NiagaraMeshRendererProperties, rendererIndex=1}
   {rendererClass=&NiagaraMeshRendererProperties, rendererIndex=2}
   {rendererClass=&NiagaraMeshRendererProperties, rendererIndex=3}
   {rendererClass=&NiagaraMeshRendererProperties, rendererIndex=4}
   {rendererClass=&NiagaraMeshRendererProperties, rendererIndex=5}
   {rendererClass=&NiagaraSpriteRendererProperties, rendererIndex=6}
  ]
 }
}
summary {
 systemName=NS_Boost_IcyBreath
 emitters [
  {emitterName=Ribbon, rendererClasses [&NiagaraRibbonRendererProperties]}
  {
   emitterName=NE_VC_Boost_Sprite_Parent
   rendererClasses [&NiagaraMeshRendererProperties, &NiagaraSpriteRendererProperties]
  }
  {emitterName=STATIC_MESH, rendererClasses [&NiagaraMeshRendererProperties, &NiagaraSpriteRendererProperties]}
 ]
 userVariables [
  User.bIsFrontendPreview {type=NiagaraBool}
  User.bIsPainted {type=NiagaraBool}
  User.ColorVariantID {type=NiagaraInt32, defaultValue=2}
  User.DTCarOffset {type=Vector3f, description="", defaultValue {struct=Vector3f, value=<0 0 0>}}
  User.GameModeID {type=NiagaraInt32}
  User.PaintedColor {type=LinearColor, description="", defaultValue {struct=LinearColor, value=<rgba 1 1 1 1>}}
  User.PaintedScalar {type=NiagaraFloat, defaultValue=1, description=""}
  User.PortCount {type=NiagaraInt32, defaultValue=2}
 ]
}
